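# Unlit seed material for the Water plugin's fluid simulation, apparently a base for supplying initial water state.
# It allows negative emissive values, which suits writing signed data rather than visible color.
# Its only sampler is the WaterVelocityAndHeight texture parameter, read as linear color at default UVs with a black 1x1 EXR default.
# No link from that sample to a material output appears in the graph.
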Material /Water/FluidSimulation/Materials/Seed/M_SeedWater_01
graph {
 properties {bAllowNegativeEmissiveColor=true, pixelDepthOffsetMode=PDOM_Legacy, shadingModel=MSM_Unlit, usedShadingModels=Unlit}
 nodes [TextureSampleParameter2D(UVs=TextureCoordinate()) {
   parameterName=WaterVelocityAndHeight
   samplerType=SAMPLERTYPE_LinearColor
   texture=/Engine/Functions/Engine_MaterialFunctions02/PivotPainter2/Black_1x1_EXR_Texture
  }]
}
